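# Sound cue for demolition impacts on a dumpster, routed through the impacts sound class.
# A Random node picks one of three equally weighted dumpster impact waves, and a Modulator applies slight random pitch variation to each hit.
# Playback uses small-object demolition hit attenuation at reduced volume and references an object-impact concurrency set.
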SoundCue /Game/Sounds/Fort_Impact_Sounds/Demolition/Dumpster/Demo_Dumpster_Impact_Cue
root=Modulator_10
nodes [
 {id=Modulator_10, children [Random_22], properties {pitchMax=1.07, pitchMin=0.92}}
 {id=Random_22, children [WavePlayer_75, WavePlayer_76, WavePlayer_78], properties {weights [1,1,1]}}
 {id=WavePlayer_75, properties {soundWaveAssetPtr=&/Dumpster_Impact01}}
 {id=WavePlayer_76, properties {soundWaveAssetPtr=&/Dumpster_Impact02}}
 {id=WavePlayer_78, properties {soundWaveAssetPtr=&/Dumpster_Impact04}}
]
properties {
 attenuationSettings=/Game/Sounds/Attenuation/Fort_Demolition_Hit_Small_Sized_Object_Att
 maxDistance=4000
 soundClassObject=&Fort_Impacts
 volumeMultiplier=0.68
 concurrencyOverrides {resolutionRule=PreventNew}
 concurrencySet [&CONC_Object_Impacts_FourMax]
}
